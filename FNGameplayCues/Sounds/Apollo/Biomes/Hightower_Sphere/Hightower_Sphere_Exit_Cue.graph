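# Sound cue that plays the exit sound for the Hightower Sphere biome in Apollo, using a bubble exit sample.
# A single wave player feeds the Hightower gadget bubble exit wave straight to the output, with no randomization or mixing.
# It uses the shared sphere enter/exit attenuation and the outdoor ambient sound class at a reduced volume.
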SoundCue /FNGameplayCues/Sounds/Apollo/Biomes/Hightower_Sphere/Hightower_Sphere_Exit_Cue
root=WavePlayer_2
nodes [{id=WavePlayer_2, properties {soundWaveAssetPtr=&/Hightower_Gadget_ST_Bubble_Exit_01}}]
properties {
 attenuationSettings=&/Hightower_Sphere_EnterExit_Att
 maxDistance=5000
 soundClassObject=&Fort_Ambient_Outdoors_Athena
 volumeMultiplier=0.626
 preEffectBusSends [{}]
}
